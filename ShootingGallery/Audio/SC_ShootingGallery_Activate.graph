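# Sound cue for the Shooting Gallery's activation sound.
# A single wave player feeds the ShootingGallery_Activate wave straight to the output at boosted volume, with no randomization or modulation.
# It borrows small melee weapon attenuation and the melee weapon sound class, and uses the shared Creative MilitaryDoor "OnlyOne" concurrency set.
# It also stores per-owner, single-instance concurrency override values, though the flag enabling that override is not set.
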SoundCue /ShootingGallery/Audio/SC_ShootingGallery_Activate
root=WavePlayer_5
nodes [{id=WavePlayer_5, properties {soundWaveAssetPtr=&/ShootingGallery_Activate}}]
properties {
 attenuationSettings=&Fort_Weapon_Melee_Sml_Att
 maxDistance=2800
 soundClassObject=&Fort_WeaponMelee_PSM
 volumeMultiplier=1.5
 concurrencyOverrides {bLimitToOwner=true, maxCount=1, resolutionRule=PreventNew, retriggerTime=0.2}
 concurrencySet [/Game/Sounds/Creative/Concurrency/MilitaryDoor_OnlyOne_Conc]
}
