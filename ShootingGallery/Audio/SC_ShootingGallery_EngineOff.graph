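# Sound cue for the shooting gallery's engine-off one-shot, reusing a clay pigeon stop sound.
# A single wave player feeds a Modulator at default settings, with the cue's base pitch lowered.
# It plays in the ambient one-shots sound class with medium prop-emitter attenuation and shared exterior ambient concurrency.
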SoundCue /ShootingGallery/Audio/SC_ShootingGallery_EngineOff
root=Modulator_0
nodes [{id=Modulator_0, children [WavePlayer_1]}, {id=WavePlayer_1, properties {soundWaveAssetPtr=/Game/Athena/Sounds/Quests/Clay_Pigeon_Stop_01}}]
properties {
 attenuationSettings=/Game/Sounds/Attenuation/Fort_Amb_Emitter_Prop_Med_Att
 maxDistance=900
 pitchMultiplier=0.7
 soundClassObject=&Fort_Ambient_OneShots
 volumeMultiplier=1
 concurrencySet [/Game/Sounds/Apollo/Biomes/Interior_Banks/AmbientAudio_Exterior_Conc]
}
