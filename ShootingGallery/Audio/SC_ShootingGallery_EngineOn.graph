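# Sound cue for the shooting gallery engine switching on and running, played as a spatialized large-prop ambient sound.
# A mixer layers a one-shot clay pigeon start sound with two loops: the gallery engine loop and a clay pigeon loop.
# An enveloper fades the clay pigeon loop in over a short ramp, and a modulator wraps the whole mix.
# A limit-to-one, prevent-new concurrency rule blocks overlapping instances.
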SoundCue /ShootingGallery/Audio/SC_ShootingGallery_EngineOn
root=Modulator_0
nodes [
 {
  id=Enveloper_1
  children [WavePlayer_5]
  properties {pitchCurve {editorCurveData {keys [{time=0}]}}, volumeCurve {editorCurveData {keys [<at 0 0.1>,<at 0.4 1>]}}}
 }
 {id=Mixer_1, children [WavePlayer_4,Enveloper_1,WavePlayer_0], properties {inputVolume [3,0.8,5]}}
 {id=Modulator_0, children [Mixer_1]}
 {id=WavePlayer_0, properties {bLooping=true, soundWaveAssetPtr=&/ShootingGallery_EngineLoop}}
 {id=WavePlayer_4, properties {soundWaveAssetPtr=/Game/Athena/Sounds/Quests/Clay_Pigeon_Start}}
 {id=WavePlayer_5, properties {bLooping=true, soundWaveAssetPtr=/Game/Athena/Sounds/Quests/Clay_Pigeon_Loop}}
]
properties {
 attenuationSettings=/Game/Sounds/Attenuation/Fort_Amb_Emitter_Prop_Large_Att
 maxDistance=2000
 soundClassObject=&Fort_Ambient_Props_Athena
 volumeMultiplier=1.6
 attenuationOverrides {falloffDistance=1000, attenuationShapeExtents {x=2000}}
 concurrencyOverrides {resolutionRule=PreventNew}
 concurrencySet [/Game/Sounds/Concurrency_Rules/LimitToOnePreventNew]
}
