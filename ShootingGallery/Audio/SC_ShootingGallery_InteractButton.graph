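# Sound cue for pressing the interact button in the Shooting Gallery.
# A single wave player feeds a modulator, routed through melee weapon sound class and small melee attenuation.
# A borrowed military door concurrency set governs playback, alongside override settings that limit to one instance per owner, prevent new plays and set a retrigger time.
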SoundCue /ShootingGallery/Audio/SC_ShootingGallery_InteractButton
root=Modulator_0
nodes [
 {id=Modulator_0, children [WavePlayer_0]}
 {id=WavePlayer_0, properties {soundWaveAssetPtr=&/ShootingGallery_InteractButton}}
]
properties {
 attenuationSettings=&Fort_Weapon_Melee_Sml_Att
 maxDistance=2800
 soundClassObject=&Fort_WeaponMelee_PSM
 volumeMultiplier=1
 concurrencyOverrides {bLimitToOwner=true, maxCount=1, resolutionRule=PreventNew, retriggerTime=0.2}
 concurrencySet [/Game/Sounds/Creative/Concurrency/MilitaryDoor_OnlyOne_Conc]
}
