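# Sound cue that plays the ShootingGallery_Stop sound for the Shooting Gallery content.
# A single wave player feeds a modulator at the root, with a boosted cue volume multiplier.
# It uses the Fort_Impacts sound class and borrows bullet impact attenuation for distance falloff.
# A shared concurrency rule named ThreeMaximum_StopFurthest governs simultaneous instances.
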SoundCue /ShootingGallery/Audio/SC_ShootingGallery_Stop
root=Modulator_0
nodes [{id=Modulator_0, children [WavePlayer_0]}, {id=WavePlayer_0, properties {soundWaveAssetPtr=&/ShootingGallery_Stop}}]
properties {
 attenuationSettings=/Game/Sounds/Attenuation/Fort_Bullet_Impact_Att
 maxDistance=5300
 soundClassObject=&Fort_Impacts
 volumeMultiplier=6
 concurrencySet [/Game/Sounds/Concurrency_Rules/ThreeMaximum_StopFurthest]
}
