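# Sound cue for a shooting gallery target going down.
# A random node picks one of three TargetDown wave variants with equal weight, and a modulator varies its pitch slightly around normal.
# The cue plays boosted in the in-game mission sound class, with borrowed gadget-activate attenuation and a concurrency rule that caps it at three voices by stopping the furthest.
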SoundCue /ShootingGallery/Audio/SC_ShootingGallery_TargetDown
root=Modulator_0
nodes [
 {id=Modulator_0, children [Random_1], properties {pitchMax=1.1, pitchMin=0.9}}
 {id=Random_1, children [WavePlayer_0,WavePlayer_1,WavePlayer_2], properties {weights [1,1,1]}}
 {id=WavePlayer_0, properties {soundWaveAssetPtr=&/ShootingGallery_TargetDown_03}}
 {id=WavePlayer_1, properties {soundWaveAssetPtr=&/ShootingGallery_TargetDown_02}}
 {id=WavePlayer_2, properties {soundWaveAssetPtr=&/ShootingGallery_TargetDown_01}}
]
properties {
 attenuationSettings=&Fort_GadgetActivate_Att
 maxDistance=4000
 soundClassObject=/Game/Sounds/FortSoundClassesAndModes/Fort_Mission_IngameSounds
 volumeMultiplier=4
 concurrencySet [/Game/Sounds/Concurrency_Rules/ThreeMaximum_StopFurthest]
}
