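# Sound cue for a shooting gallery target popping up, playing one of five TargetUp recordings.
# An equally weighted Random node picks the wave, and a root Modulator varies its pitch slightly for variety.
# It borrows bullet impact attenuation and the impacts sound class, with a boosted volume multiplier.
# A shared concurrency rule, named for a three-sound limit that stops the furthest, governs overlapping plays.
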SoundCue /ShootingGallery/Audio/SC_ShootingGallery_TargetUp
root=Modulator_0
nodes [
 {id=Modulator_0, children [Random_0], properties {pitchMax=1.1, pitchMin=0.9}}
 {id=Random_0, children [WavePlayer_4,WavePlayer_5,WavePlayer_6,WavePlayer_7,WavePlayer_8], properties {weights [1,1,1,1,1]}}
 {id=WavePlayer_4, properties {soundWaveAssetPtr=&/ShootingGallery_TargetUp_01}}
 {id=WavePlayer_5, properties {soundWaveAssetPtr=&/ShootingGallery_TargetUp_02}}
 {id=WavePlayer_6, properties {soundWaveAssetPtr=&/ShootingGallery_TargetUp_03}}
 {id=WavePlayer_7, properties {soundWaveAssetPtr=&/ShootingGallery_TargetUp_04}}
 {id=WavePlayer_8, properties {soundWaveAssetPtr=&/ShootingGallery_TargetUp_05}}
]
properties {
 attenuationSettings=/Game/Sounds/Attenuation/Fort_Bullet_Impact_Att
 maxDistance=5300
 soundClassObject=&Fort_Impacts
 volumeMultiplier=7
 concurrencySet [/Game/Sounds/Concurrency_Rules/ThreeMaximum_StopFurthest]
}
