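# Sound cue for hits on friendly targets in the shooting gallery.
# It randomly picks one of three short impact waves with equal weight, then a modulator slightly varies the volume.
# It references bullet-impact attenuation settings and the enemy-damaged sound class.
# A concurrency override caps simultaneous instances and stops the oldest when the limit is reached.
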SoundCue /ShootingGallery/Audio/SC_ShootingGallery_Target_Impact_Friendly
root=Modulator_0
nodes [
 {id=Modulator_0, children [Random_0], properties {volumeMax=1.1, volumeMin=0.9}}
 {id=Random_0, children [WavePlayer_3,WavePlayer_4,WavePlayer_5], properties {weights [1,1,1]}}
 {id=WavePlayer_3, properties {soundWaveAssetPtr=&/Target_Impact_Friendly_Short_01}}
 {id=WavePlayer_4, properties {soundWaveAssetPtr=&/Target_Impact_Friendly_Short_02}}
 {id=WavePlayer_5, properties {soundWaveAssetPtr=&/Target_Impact_Friendly_Short_03}}
]
properties {
 attenuationSettings=/Game/Sounds/Attenuation/Fort_Bullet_Impact_Athena_Att
 bOverrideConcurrency=true
 maxDistance=4500
 soundClassObject=/Game/Sounds/FortSoundClassesAndModes/Fort_Enemy_Damaged_PSM
 volumeMultiplier=1.2
 attenuationOverrides {falloffDistance=1600}
 concurrencyOverrides {maxCount=5, resolutionRule=StopOldest}
}
